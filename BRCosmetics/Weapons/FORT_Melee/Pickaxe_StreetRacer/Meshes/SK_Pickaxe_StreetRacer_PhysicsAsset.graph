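# Physics asset for the StreetRacer two-handed pickaxe skeletal mesh, defining collision and joints for its handle and chain bones.
# A long capsule covers the handle, and smaller capsules cover chain_start and dyn_chain_02.
# Both chain bodies are constrained directly to the handle rather than linked in series.
# The dyn_chain_02 joint locks twist and one swing axis, leaving a single hinge-like swing axis.
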PhysicsAsset /BRCosmetics/Weapons/FORT_Melee/Pickaxe_StreetRacer/Meshes/SK_Pickaxe_StreetRacer_PhysicsAsset
bodies {
 Melee_TwoHanded_Handle [{shapeType=Capsule, length=88.1946, radius=3.7014, center=<-0.0001 -1.5727 21.477>, rotation=<rot 0 -90 0>}]
 chain_start [{shapeType=Capsule, length=4.0048, radius=3.9742, center=<1 0 0>}]
 dyn_chain_02 [{shapeType=Capsule, length=7.9825, radius=5.5682, center=<7.0333 -0.0055 5.8158>, rotation=<rot 56.2446 -180 -180>}]
}
constraints [
 {bone1Name=chain_start, bone2Name=Melee_TwoHanded_Handle, swing1LimitDegrees=15, swing2LimitDegrees=60, twistLimitDegrees=10}
 {
  bone1Name=dyn_chain_02
  bone2Name=Melee_TwoHanded_Handle
  swing1LimitDegrees=0
  swing1Motion=Locked
  swing2LimitDegrees=66
  twistLimitDegrees=45
  twistMotion=Locked
 }
]
